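; AI controller component for the CoffeePuff NPC that supplies rules for random target selection.
; Target switching is always allowed unless the bot's alert level is LostTarget or Threatened, when it requires the pawn's EnergyDepleted gameplay tag.
; Downed (DBNO) targets are ignored, and targets outside the bot controller's leash are undesirable.
; Holster application is always refused, and the BeginPlay and Tick events are empty.
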(event EventBeginPlay)

(event EventTick
  (params (DeltaSeconds :type Float%20%28single-precision%29)))

(fn AllowedToSwitchTarget
  (bind _asfort_athena_aibot_controller (Utilities|Casting|CastToFortAthenaAIBotController (Components|GetOwner)))
  (bind _output_get (LocalVariables|SetOwnerAIBotController _asfort_athena_aibot_controller))
  (bind _asfort_player_pawn (Utilities|Casting|CastToFortPlayerPawn (Pawn|GetControlledPawn _output_get)))
  (LocalVariables|SetOwnerPawn _asfort_player_pawn)
  (switch Utilities|FlowControl|Switch|SwitchonEAlertLevel (Variables|/Script/FortniteAI.FortAthenaAIBotController|GetCurrentAlertLevel (LocalVariables|GetOwnerAIBotController))
    (:Unaware
      _)
    (:Suspicious
      _)
    (:LostTarget
      (if (GameplayTags|HasMatchingGameplayTag (LocalVariables|GetOwnerPawn) "(TagName=\"Event.Ability.VividRazor.CoffeePuff.EnergyDepleted\")") :!label _label
        (return true)
        (else
          (return false))))
    (:Threatened
      (goto _label))
    (:Count
      _))
  (return true))

(fn CanApplyHolster
  (params (HolsterId :type Name%20%28by%20ref%29 :default "None"))
  (return false))

(fn TargetIsIgnored
  (params (Target :type Actor%20Object%20Reference))
  (bind _asfort_pawn (Utilities|Casting|CastToFortPawn Target))
  (return (Health|IsDBNO _asfort_pawn)))

(fn TargetIsUndesirable
  (params (Target :type Actor%20Object%20Reference))
  (bind _asfort_athena_aibot_controller (Utilities|Casting|CastToFortAthenaAIBotController (Components|GetOwner)))
  (return (not (Class|FortAthenaAIBotController|IsLocationInsideLeash _asfort_athena_aibot_controller (Transformation|GetActorLocation Target) true 0.0))))
